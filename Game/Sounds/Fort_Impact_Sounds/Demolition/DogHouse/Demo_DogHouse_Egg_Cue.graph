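# Sound cue for the impact of a doghouse ant egg, part of the demolition impact sounds.
# A random node picks one of four equally weighted impact waves and feeds a modulator that slightly varies pitch on each play.
# The cue uses small melee weapon attenuation, the impacts sound class, a slight volume boost and an object-impact concurrency set.
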SoundCue /Game/Sounds/Fort_Impact_Sounds/Demolition/DogHouse/Demo_DogHouse_Egg_Cue
root=Modulator_2
nodes [
 {id=Modulator_2, children [Random_0], properties {pitchMax=1.07, pitchMin=0.92}}
 {id=Random_0, children [WavePlayer_0,WavePlayer_1,WavePlayer_2,WavePlayer_3], properties {weights [1,1,1,1]}}
 {id=WavePlayer_0, properties {soundWaveAssetPtr=&/DogHouse_AntEgg_Impact_01}}
 {id=WavePlayer_1, properties {soundWaveAssetPtr=&/DogHouse_AntEgg_Impact_02}}
 {id=WavePlayer_2, properties {soundWaveAssetPtr=&/DogHouse_AntEgg_Impact_03}}
 {id=WavePlayer_3, properties {soundWaveAssetPtr=&/DogHouse_AntEgg_Impact_04}}
]
properties {
 attenuationSettings=&Fort_Weapon_Melee_Sml_Att
 maxDistance=2800
 soundClassObject=&Fort_Impacts
 volumeMultiplier=1.2
 concurrencyOverrides {resolutionRule=PreventNew}
 concurrencySet [&CONC_Object_Impacts_FourMax]
}
